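# Additive emissive material for the RL83Holo holographic wheel, tinted by CustomColor and usable on skeletal meshes.
# An animated flipbook atlas drives the image, with optional chromatic aberration sampling it at offset UVs per color channel and an optional scrolling screen-space scanline.
# Vertex color blue blends that result toward a named-reroute input.
# Named reroutes also declare procedural striplines, flicker and periodic block-texture UV glitches; the Glitches switch output is left unconnected.
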Material /VehicleCosmetics/Wheels/RL83Holo/Materials/M_Wheel_RL83Holo
graph {
 properties {
  bUseMaterialAttributes=true
  bUsedWithMeshDeformer=true
  bUsedWithSkeletalMesh=true
  blendMode=BLEND_Additive
  pixelDepthOffsetMode=PDOM_Legacy
  usedShadingModels=Default Lit
 }
 nodes [
  n1 = ScreenPosition()
  n2 = Frac(None=Add(A=ComponentMask(None=@n1.ViewportUV) {g=true}, B=Time()))
  n3 = OneMinus(None=@n2)
  n4 = MaterialFunctionCall(
   "Animation  Phase (0-1)"=Multiply(A=Time(), B=Constant() {r=0.5})
   Number of Rows=Constant() {r=4}
   Number of Columns=Constant() {r=8}
   UVs=TextureCoordinate()
  ) {materialFunction=&FlipBook}
  n5 = Constant() {r=0.35}
  n6 = VectorParameter() {group=5_Special Effects, parameterName=CustomColor, defaultValue=<rgba 0.23 0.9 1 1>}
  n7 = Multiply(
   A=@n6.R
   B=Multiply(
    A=TextureSample(UVs=@n4.UVs) {samplerType=SAMPLERTYPE_Masks, texture=/VehicleCosmetics/Wheels/RL83Holo/Textures/Holo_sequence_atlas}.R
    B=@n5
   )
  )
  n8 = Time()
  n9 = Constant() {r=0.5}
  n10 = TextureCoordinate()
  n11 = ScalarParameter() {parameterName=Chroma_distance, sliderMax=1}
  n12 = Multiply(A=Constant2Vector() {r=-0.1}, B=@n11)
  n13 = TextureSample(UVs=MaterialFunctionCall(
    "Animation  Phase (0-1)"=Multiply(A=@n8, B=@n9)
    Number of Rows=Constant() {r=4}
    Number of Columns=Constant() {r=8}
    UVs=Add(A=@n10, B=@n12)
   ) {materialFunction=&FlipBook}.UVs) {samplerType=SAMPLERTYPE_Masks, texture=&Holo_sequence_atlas}
  n14 = NamedRerouteUsage()
  n15 = Multiply(A=Add(A=@n13.R, B=ComponentMask(None=@n14) {r=true}), B=@n13.R)
  n16 = Multiply(A=Multiply(A=@n6.G, B=Multiply(A=@n15, B=@n5)), B=NamedRerouteUsage())
  n17 = Time()
  n18 = Constant() {r=0.5}
  n19 = TextureCoordinate()
  n20 = Multiply(A=@n11, B=Constant2Vector() {g=0.1})
  n21 = TextureSample(UVs=MaterialFunctionCall(
    "Animation  Phase (0-1)"=Multiply(A=@n17, B=@n18)
    Number of Rows=Constant() {r=4}
    Number of Columns=Constant() {r=8}
    UVs=Add(A=@n19, B=@n20)
   ) {materialFunction=&FlipBook}.UVs) {samplerType=SAMPLERTYPE_Masks, texture=&Holo_sequence_atlas}
  n22 = ComponentMask(None=NamedRerouteUsage()) {r=true}
  n23 = Multiply(A=Multiply(A=Add(A=@n21.R, B=@n22), B=@n21.R), B=@n5) {constB=0.33}
  n24 = ScalarParameter() {defaultValue=1, group=5_Special Effects, parameterName=EmissiveStrength}
  n25 = StaticSwitchParameter(
   True=AppendVector(A=AppendVector(A=@n7, B=@n16), B=Multiply(A=@n6.B, B=@n23))
   False=Multiply(
    A=Multiply(A=@n24, B=@n6.RGB)
    B=TextureSample(UVs=@n4.UVs) {samplerType=SAMPLERTYPE_Masks, texture=&Holo_sequence_atlas}.R
   )
  ) {defaultValue=true, parameterName=ChromaticAbberation}
  n26 = Multiply(A=Multiply(A=Constant() {r=4}, B=Multiply(A=@n2, B=@n3)), B=@n25)
  MakeMaterialAttributes!(
   BaseColor
   Metallic
   Specular
   Roughness
   Anisotropy
   EmissiveColor=LinearInterpolate(A=StaticSwitchParameter(True=@n26, False=@n25) {parameterName=Scanline}, B=NamedRerouteUsage(), Alpha=VertexColor().B)
   Opacity
   OpacityMask
   Normal
   Tangent
   WorldPositionOffset
   SubsurfaceColor
   ClearCoat
   ClearCoatRoughness
   AmbientOcclusion
   Refraction
   CustomizedUVs_0
   CustomizedUVs_1
   CustomizedUVs_2
   CustomizedUVs_3
   CustomizedUVs_4
   CustomizedUVs_5
   CustomizedUVs_6
   CustomizedUVs_7
   PixelDepthOffset
   ShadingModel
   Displacement
  )
  n27 = Multiply(B=Time()) {constA=1.5}
  n28 = TextureCoordinate() {uTiling=0.5, vTiling=0.4}
  n29 = ComponentMask(None=TextureSample(UVs=Panner(Coordinate=@n28, Time=@n27) {speedX=0.5, speedY=-0.2}) {samplerType=SAMPLERTYPE_LinearColor, texture=/VehicleCosmetics/Wheels/RL83Holo/Textures/blocks_Tex_b_alt}.RGB) {g=true, r=true}
  n30 = Constant() {r=2}
  n31 = TextureCoordinate()
  n32 = If(A=Frac(None=@n27), "A > B"=Add(A=Divide(A=@n29, B=@n30), B=@n31), "A == B"=@n31, "A < B"=@n31) {constB=0.9}
  NamedRerouteDeclaration(None=@n32) {name=Glitches_output, nodeColor=<rgba 0.2891 0 1 1>}
  NamedRerouteDeclaration(None=@n6.RGB) {name=CustomColor_output, nodeColor=<rgba 1 0.0469 0 1>}
  n33 = ComponentMask(None=TextureCoordinate() {uTiling=0.5, vTiling=0.4}) {r=true}
  n34 = ScalarParameter() {defaultValue=30, group=5_Special Effects, parameterName=Stripline_Intensity}
  n35 = Time()
  n36 = Constant() {r=2}
  n37 = Frac(None=Add(A=Multiply(A=@n33, B=@n34), B=Multiply(A=@n35, B=@n36)))
  n38 = Multiply(A=Multiply(A=@n37, B=OneMinus(None=@n37)), B=Constant() {r=2})
  n39 = ComponentMask(None=@n38) {r=true}
  n40 = ComponentMask(None=@n38) {g=true}
  n41 = TextureCoordinate() {uTiling=0.5, vTiling=0.4}
  n42 = Power(Base=OneMinus(None=ComponentMask(None=@n41) {g=true})) {constExponent=12}
  n43 = Multiply(A=Multiply(A=Add(A=@n39, B=@n40), B=Clamp(None=@n42)), B=NamedRerouteUsage())
  NamedRerouteDeclaration(None=Multiply(A=NamedRerouteUsage(), B=@n43)) {name=Stripline_output, nodeColor=<rgba 0 0.2734 1 1>}
  NamedRerouteDeclaration(None=@n43) {name=Stripline_mask_output, nodeColor=<rgba 1 0 0.4453 1>}
  NamedRerouteDeclaration(None=@n16) {name=holo_mask_output, nodeColor=<rgba 0 1 0.0156 1>}
  n44 = Time()
  n45 = Multiply(B=@n44) {constA=32}
  n46 = Add(A=@n45) {constB=0.4512}
  n47 = Multiply(A=Frac(None=Sine(None=@n45)), B=Frac(None=Sine(None=@n46) {period=2}))
  n48 = Multiply(A=@n44)
  n49 = Multiply(A=Add(A=Sine(None=@n48))) {constB=512}
  NamedRerouteDeclaration(None=LinearInterpolate(A=Frac(None=@n47), Alpha=Clamp(None=@n49))) {name=Flicker_output, nodeColor=<rgba 0 1 0.3672 1>}
  StaticSwitchParameter(True=@n32, False=@n31) {group=5_Special Effects, parameterName=Glitches}
 ]
}
